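# Niagara system for the Toon vehicle boost cosmetic, emitting cartoon-style exhaust from the car's skeletal mesh sockets, offset by DTCarOffset.
# A ribbon emitter draws layered flames per port, while a sprite emitter adds growing SubUV flipbook puffs.
# Another sprite emitter uses two spawn groups with different motion, split by VisibilityTag, feeding four renderers with outline and front/back star parameters.
# Painted-variant modules set per-color palettes, with extra R1–R3 layer colors on some emitters.
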NiagaraSystem /VehicleCosmetics/Boosts/Boost_Toon/FX/NS_Boost_Toon
dependencies {
 usedDataInterfaces [
  &NiagaraDataInterfaceArrayFloat
  &NiagaraDataInterfaceCurve
  &NiagaraDataInterfaceArrayFloat3
  &NiagaraDataInterfaceSpriteRendererInfo
  &NiagaraDataInterfaceColorCurve
  &NiagaraDataInterfaceSkeletalMesh
 ]
 usedDynamicInputs []
 usedModules [
  &self:SystemSpawnScript.SystemScriptSource.SystemScriptGraph.NiagaraNodeAssignment_0.SetVariables_#1
  &SystemState
 ]
}
emitters {
 NE_VC_Boost_Ribbon_Flame {
  emitterUpdateScript [
   &EmitterState()
   &VC_FilteredSocketCount()
   &VC_RateQualityHandler(Epic Spawn Rate=120, High Spawn Rate=60, Low Spawn Rate=30, Use Custom Rate?=true)
   &VC_SpawnRibbon_v1(Number of Ribbons=$Emitter.PortCount, SpawnRate=$Emitter.Spawn Rate)
  ]
  particleSpawnScript [
   &VC_PaintedVariantsOverride(
    BurntSienna=<rgba 0.3 0.0926 0.037 2>
    Cobalt=<rgba 0 0.0148 0.8 1>
    Crimson=<rgba 0.4635 0 0.0126 1>
    DefaultColor=<rgba 0.7289 0.243 0 1>
    Grey=<rgba 0.35 0.35 0.35 1>
    Lime=<rgba 0.4414 0.5885 0 1>
    Orange=<rgba 0.45 0.3058 0.0109 1>
    Purple=<rgba 0.3867 0 0.4 1>
    R1 Black:LinearColor=<rgba 0.0191 0.0191 0.0191 2>
    R1 BurntSienna:LinearColor=<rgba 2 0.6 0 5>
    R1 Cobalt:LinearColor=<rgba 0 0.0925 5 5>
    R1 Crimson:LinearColor=<rgba 2 0 0.0542 5>
    R1 DefaultColor=<rgba 2 1.7805 0.3947 5>
    R1 ForestGreen:LinearColor=<rgba 0 2 0.1011 5>
    R1 Grey:LinearColor=<rgba 3 3 3 5>
    R1 Lime:LinearColor=<rgba 1.6667 2 0 5>
    R1 Orange:LinearColor=<rgba 2 1.6667 0 5>
    R1 Pink:LinearColor=<rgba 2.5 0.2995 1.6998 5>
    R1 Purple:LinearColor=<rgba 1.5467 0 1.6 4>
    R1 Saffron:LinearColor=<rgba 1.5 1.5 0 5>
    R1 SkyBlue:LinearColor=<rgba 0 1.9962 3 5>
    R1 TitaniumWhite:LinearColor=<rgba 2 2 2 5>
    R2 Black:LinearColor=<rgba 0.0191 0.0191 0.0191 2>
    R2 BurntSienna:LinearColor=<rgba 0.5 0.1544 0.0616 3>
    R2 Crimson:LinearColor=<rgba 0.6 0 0.0163 2>
    R2 DefaultColor:LinearColor=<rgba 0.634 0.3683 0.089 3>
    R2 ForestGreen:LinearColor=<rgba 0 0.5 0.0253 2>
    R2 Grey:LinearColor=<rgba 0.3 0.3 0.3 2>
    R2 Lime:LinearColor=<rgba 0.4414 0.5885 0 2>
    R2 Orange:LinearColor=<rgba 0.5 0.375 0 2>
    R2 Pink:LinearColor=<rgba 0.5052 0.0605 0.3435 2>
    R2 Purple:LinearColor=<rgba 0.4333 0 0.65 1.7>
    R2 Saffron:LinearColor=<rgba 0.8 0.8 0 2>
    R2 SkyBlue:LinearColor=<rgba 0 0.3566 0.536 2>
    R2 TitaniumWhite:LinearColor=<rgba 0.5 0.5 0.5 2>
    Saffron=<rgba 0.6 0.6 0 1>
    Use R1 Painted Overrides=true
    Use R2 Painted Overrides=true
   )
   &VC_InitializeRibbon_v1(
    Color=$Particles.Color
    Custom Offset=true
    Enable Skeletal Position=true
    Lifetime=0.28
    Local Vector Offset=&AddVectorToPosition(Position=$User.DTCarOffset, Vector=<0 0 0>)
    Offset Amount=1
    Ribbon Width=100
   )
   &AddVelocity(Velocity=<-1000 0 0>) {module=AddVelocity001}
   /SetVariables(Particles.WidthRibbon3:NiagaraFloat=&SelectFloatFromArray(Direct Array Index=$Particles.SpawnGroup, Float Selection Array=DataInterface() {propertyValues {FloatData [-120,120]}})) {isSetParameters=true}
   /VehicleCosmetics/FX/VC_FXParents/Modules/DLR_ProductPreviews(GameModeID:NiagaraInt32=$System.GameModeID, bPreviewVehicleSpeed:NiagaraBool=$System.bPreviewVehicleSpeed)
  ]
  particleUpdateScript [
   &ParticleState()
   &InheritSourceMovement(ApplyForce=false, ApplyPosition=true, ApplyPositionFalloff=false) {module=InheritSourceMovement001}
   &InheritSourceMovement(
    Applied Position Scale=&VectorFromFloat(Value=&FloatFromCurve(FloatCurve=DataInterface() {propertyValues {Curve {
         defaultValue=1
         keys [
          {arriveTangent=-1.7885, interpMode=RCIM_Cubic, leaveTangent=-1.7885, tangentMode=RCTM_User, time=0.05, value=1}
          {interpMode=RCIM_Cubic, tangentMode=RCTM_User, time=1, value=0}
         ]
        }}}))
    ApplyForce=false
    ApplyPosition=true
    ApplyPositionFalloff=false
   ) {enabled=false}
   &SolveForcesAndVelocity(Clamp Velocity=false, Limit Acceleration=false)
   /SetVariables(Particles.FlamesPosition:NiagaraPosition=&AddVectorToPosition(
     Position=$Particles.Position
     Vector=/VehicleCosmetics/FX/Niagara/Modules/VC_ConvertVectorToWorld(Vector:Vector3f=&AddVectorToPosition(
       Position=&SelectVectorFromArray(
        Direct Array Index=&SpawnGroup()
        Vector Selection Array=DataInterface() {propertyValues {FloatData [<0 12 0>, <0 -12 0>]}}
       )
       Vector=Linked() {linkedVariable {name=User.DTCarOffset, type=NiagaraPosition}}
      ))
    )) {isSetParameters=true}
   /SetVariables(
    Particles.FillRibbonWidth:NiagaraFloat=&Multiply_Float(
     A=$Particles.WidthRibbon3
     B=&FloatFromCurve(FloatCurve=DataInterface() {propertyValues {Curve {keys [<at 0.5248 1>,<at 0.7154 0>]}}}, Scale Curve=0.82)
    )
    Particles.FlameRibbonWidth:NiagaraFloat=&Multiply_Float(
     A=$Particles.WidthRibbon3
     B=&FloatFromCurve(FloatCurve=DataInterface() {propertyValues {Curve {keys [<at 0.3123 1>,<at 0.9347 0>]}}})
    )
   ) {isSetParameters=true}
   &VC_UpdateRibbon_v1(Enable Skeletal Position=true)
  ]
  rendererClasses [&NiagaraRibbonRendererProperties]
  renderers [
   {rendererClass=&NiagaraRibbonRendererProperties}
   {rendererClass=&NiagaraRibbonRendererProperties, rendererIndex=1}
   {rendererClass=&NiagaraRibbonRendererProperties, rendererIndex=2}
  ]
 }
 NE_VC_Boost_Sprite_Parent {
  emitterUpdateScript [
   &EmitterState()
   &VC_FilteredSocketCount()
   &SpawnRate(SpawnRate=16, Use Spawn Probability=false)
  ]
  particleSpawnScript [
   &VC_PaintedVariantsOverride(
    BurntSienna=<rgba 0.4 0.1235 0.0493 1>
    Cobalt=<rgba 0 0.0185 1 1>
    DefaultColor=&ColorFromCurve(ColorCurve=DataInterface() {propertyValues {
       AlphaCurve {keys [<at 0 1 RCIM_Cubic>]}
       BlueCurve {keys [<at 0 0.0484 RCIM_Cubic>,<at 0.7477 0.0432 RCIM_Cubic>]}
       GreenCurve {keys [<at 0 0.2446 RCIM_Cubic>,<at 0.7477 0.1481 RCIM_Cubic>]}
       RedCurve {keys [<at 0 0.9713 RCIM_Cubic>,<at 0.7477 0.3652 RCIM_Cubic>]}
      }})
    Grey=<rgba 0.3 0.3 0.3 1>
    Lime=<rgba 0.3924 0.5885 0 1>
    Orange=<rgba 0.65 0.4333 0 1>
    Pink=<rgba 0.5052 0.0605 0.3435 1>
    Purple=<rgba 0.3398 0 0.6795 1>
    Saffron=<rgba 0.5149 0.5149 0.016 1>
    TitaniumWhite=<rgba 0.6 0.6 0.6 1>
   ) {enabled=false}
   &InitializeParticle(
    Color Mode=::NewEnumerator0
    Lifetime=0.55
    Sprite Rotation Mode=::NewEnumerator2
    Sprite Size Mode=::NewEnumerator3
    Uniform Sprite Size=40
    UsePositionOffset=false
    Write Material Random=true
   )
   &PartitionParticles(Number Of Partitions=$Emitter.PortCount)
   &SkeletalMeshLocation(
    Bone / Socket Index=$Particles.PartitionParticles.Partition
    Bone / Socket Sampling Mode=::NewEnumerator5
    MaskSpawnGroup=false
    Mesh Sampling Type=::NewEnumerator6
    Sampled Position Offset=&AddVectorToPosition(Position=<-10 0 0>, Vector=Linked() {linkedVariable {name=User.DTCarOffset, type=NiagaraPosition}})
    Skeletal Mesh=$System.SkeletalMesh
   )
   &AddVelocity(Velocity=<-150 0 0>)
   &DLR_ProductPreviews(GameModeID:NiagaraInt32=$System.GameModeID, bPreviewVehicleSpeed:NiagaraBool=$System.bPreviewVehicleSpeed)
  ]
  particleUpdateScript [
   &ParticleState()
   &VC_PaintedVariantsOverride(
    BurntSienna=<rgba 0.4 0.1235 0.0493 1>
    Cobalt=<rgba 0 0.0185 1 1>
    DefaultColor=&ColorFromCurve(ColorCurve=DataInterface() {propertyValues {
       AlphaCurve {keys [<at 0 1 RCIM_Cubic>]}
       BlueCurve {keys [<at 0 0.0484 RCIM_Cubic>,<at 0.7477 0.0432 RCIM_Cubic>]}
       GreenCurve {keys [<at 0 0.2446 RCIM_Cubic>,<at 0.7477 0.1481 RCIM_Cubic>]}
       RedCurve {keys [<at 0 0.9713 RCIM_Cubic>,<at 0.7477 0.3652 RCIM_Cubic>]}
      }})
    Grey=<rgba 0.3 0.3 0.3 1>
    Lime=<rgba 0.3924 0.5885 0 1>
    Orange=<rgba 0.65 0.4333 0 1>
    Pink=<rgba 0.5052 0.0605 0.3435 1>
    Purple=<rgba 0.3398 0 0.6795 1>
    Saffron=<rgba 0.5149 0.5149 0.016 1>
    TitaniumWhite=<rgba 0.6 0.6 0.6 1>
   ) {module=VC_PaintedVariantsOverride001}
   &ScaleSpriteSize(
    Uniform Curve Scale=1.35
    Uniform Curve Sprite Scale=DataInterface() {propertyValues {Curve {keys [{tangentMode=RCTM_User, time=0, value=1}, {tangentMode=RCTM_User, time=1, value=2.5}]}}}
   )
   &SubUVAnimation(
    "Sprite Renderer "=DataInterface() {propertyValues {SpriteRenderer=&self:NE_VC_Boost_Sprite_Parent.NiagaraSpriteRendererProperties_5}}
    SubUV Animation Mode:ENiagara_SubUVLookupModeV2=::NewEnumerator1
    SubUV Curve=DataInterface() {propertyValues {Curve {
       defaultValue=0
       keys [
        {arriveTangent=0.5133, interpMode=RCIM_Cubic, leaveTangent=0.5133, tangentMode=RCTM_User, time=0, value=0}
        {interpMode=RCIM_Cubic, tangentMode=RCTM_User, time=0.85, value=1}
       ]
      }}}
    UseEndFrame=false
    UseStartFrame=false
   )
   &InheritSourceMovement(ApplyForce=false, ApplyPosition=true, ApplyPositionFalloff=false)
   &InheritSourceMovement(
    Applied Position Scale=&Multiply_VectorByFloat(
     Vector=<1 1 1>
     Float=&FloatFromCurve(FloatCurve=DataInterface() {propertyValues {Curve {
         defaultValue=1
         keys [
          {arriveTangent=-1.7885, interpMode=RCIM_Cubic, leaveTangent=-1.7885, tangentMode=RCTM_User, time=0.05, value=1}
          {interpMode=RCIM_Cubic, tangentMode=RCTM_User, time=1, value=0}
         ]
        }}})
    )
    ApplyForce=false
    ApplyPosition=true
    ApplyPositionFalloff=false
   ) {enabled=false, module=InheritSourceMovement001}
   &SolveForcesAndVelocity(Clamp Velocity=false, Limit Acceleration=false)
  ]
  rendererClasses [&NiagaraSpriteRendererProperties]
  renderers [{rendererClass=&NiagaraSpriteRendererProperties}]
 }
 NE_VC_Boost_Sprite_Parent001 {
  emitterUpdateScript [
   &EmitterState()
   &VC_FilteredSocketCount()
   &SpawnRate(SpawnRate=30, Use Spawn Probability=false)
   &SpawnRate(SpawnGroup=1, SpawnRate=40, Use Spawn Probability=false) {module=SpawnRate002}
  ]
  particleSpawnScript [
   /SetVariables(Particles.VisibilityTag=$Emitter.SpawnGroup) {isSetParameters=true}
   &VC_PaintedVariantsOverride(
    Black=<rgba 0.0065 0.0065 0.007 1>
    BurntSienna=<rgba 0.53 0.1502 0 1>
    Cobalt=<rgba 0.0079 0.0463 0.29 1>
    Crimson=<rgba 0.28 0.0056 0.006 1>
    DefaultColor=<rgba 0.2 0.1891 0.18 1>
    ForestGreen=<rgba 0 0.14 0.0071 1>
    Gold:LinearColor=<rgba 0.1755 0.1578 0.1413 1>
    Grey=<rgba 0.158 0.158 0.16 1>
    Lime=<rgba 0.3855 0.4917 0.0131 1>
    Orange=<rgba 0.45 0.3354 0.0224 1>
    Pink=<rgba 0.38 0 0.1653 1>
    Purple=<rgba 0.1367 0.0018 0.19 1>
    R1 Black:LinearColor=<rgba 0 0 0 1>
    R1 BurntSienna:LinearColor=<rgba 0.12 0.036 0 1>
    R1 Cobalt:LinearColor=<rgba 0.0033 0.0192 0.12 1>
    R1 Crimson:LinearColor=<rgba 0.11 0.0022 0.0023 1>
    R1 DefaultColor=<rgba 0.07 0.0601 0.0529 1.5>
    R1 ForestGreen:LinearColor=<rgba 0 0.03 0.0015 1>
    R1 Grey:LinearColor=<rgba 0.05 0.05 0.05 1>
    R1 Lime:LinearColor=<rgba 0.0941 0.12 0.0032 1>
    R1 Orange:LinearColor=<rgba 0.22 0.0807 0 1>
    R1 Pink:LinearColor=<rgba 0.12 0 0.0522 1>
    R1 Purple:LinearColor=<rgba 0.0457 0 0.06 1>
    R1 Saffron:LinearColor=<rgba 0.1618 0.17 0.0068 1>
    R1 SkyBlue:LinearColor=<rgba 0 0.1131 0.17 1>
    R1 TitaniumWhite:LinearColor=<rgba 0.18 0.182 0.2 1>
    R2 Black:LinearColor=<rgba 0.1195 0.1195 0.1221 1>
    R2 BurntSienna:LinearColor=<rgba 1 0.0833 0 1>
    R2 Cobalt:LinearColor=<rgba 0.08 0.4592 0.8 1>
    R2 Crimson:LinearColor=<rgba 0.4635 0 0 1>
    R2 DefaultColor:LinearColor=<rgba 1 0.9539 0.07 1>
    R2 ForestGreen:LinearColor=<rgba 0.0233 0.7 0 1>
    R2 Grey:LinearColor=<rgba 0.4 0.4 0.4 1>
    R2 Lime:LinearColor=<rgba 0.3924 0.5885 0 1>
    R2 Orange:LinearColor=<rgba 0.6724 0.5043 0 1>
    R2 Pink:LinearColor=<rgba 0.65 0.273 0.5129 1>
    R2 Purple:LinearColor=<rgba 0.2257 0 0.55 1>
    R2 Saffron:LinearColor=<rgba 2 1.3333 0 1>
    R2 SkyBlue:LinearColor=<rgba 0.0536 0.3749 0.5365 1>
    R2 TitaniumWhite:LinearColor=<rgba 0.8 0.8 0.8 1>
    R3 Black:LinearColor=<rgba 0 0 0 1>
    R3 BurntSienna:LinearColor=<rgba 0 0 0 1>
    R3 Cobalt:LinearColor=<rgba 0 0 0 1>
    R3 Crimson:LinearColor=<rgba 0 0 0 1>
    R3 DefaultColor:LinearColor=<rgba 0.02 0.008 0 1>
    R3 ForestGreen:LinearColor=<rgba 0 0 0 1>
    R3 Grey:LinearColor=<rgba 0 0 0 1>
    R3 Lime:LinearColor=<rgba 0 0 0 1>
    R3 Orange:LinearColor=<rgba 0 0 0 1>
    R3 Pink:LinearColor=<rgba 0 0 0 1>
    R3 Purple:LinearColor=<rgba 0.0041 0 0.01 1>
    R3 Saffron:LinearColor=<rgba 0 0 0 1>
    R3 SkyBlue:LinearColor=<rgba 0 0 0 1>
    R3 TitaniumWhite:LinearColor=<rgba 0 0 0 1>
    Saffron=<rgba 0.5331 0.56 0.0224 1>
    SkyBlue=<rgba 0 0.2795 0.42 1>
    TitaniumWhite=<rgba 1 1 1 1>
    Use R1 Painted Overrides=true
    Use R2 Painted Overrides=true
    Use R3 Painted Overrides=true
   )
   &InitializeParticle(
    Color Mode=::NewEnumerator0
    Lifetime=&MergeHelper_FloatSwitch(
     0=1
     1=&RandomRangeFloat(Minimum=0.8, Maximum=1.5)
     2=?
     3=?
     Index=$Emitter.SpawnGroup
     Num=::NewEnumerator0
    )
    Sprite Rotation Angle=&MergeHelper_FloatSwitch(
     0=&RandomRangeFloat(Minimum=0, Maximum=1)
     1=&RandomRangeFloat(Minimum=0.05, Maximum=0.2)
     2=?
     3=?
     Index=$Emitter.SpawnGroup
     Num=::NewEnumerator0
    )
    Sprite Rotation Mode=::NewEnumerator1
    Sprite Size Mode=::NewEnumerator3
    Uniform Sprite Size=140
    UsePositionOffset=false
    Write Material Random=true
   )
   &PartitionParticles(Number Of Partitions=$Emitter.PortCount)
   &SkeletalMeshLocation(
    Bone / Socket Index=$Particles.PartitionParticles.Partition
    Bone / Socket Sampling Mode=::NewEnumerator5
    MaskSpawnGroup=false
    Mesh Sampling Type=::NewEnumerator6
    Sampled Position Offset=&AddVectorToPosition(Position=<-20 0 0>, Vector=Linked() {linkedVariable {name=User.DTCarOffset, type=NiagaraPosition}})
    Skeletal Mesh=$System.SkeletalMesh
   )
   &ShapeLocation(
    MaskSpawnGroup=false
    Override Seed=false
    Sphere Radius=&MergeHelper_FloatSwitch(0=30, 1=0, 2=?, 3=?, Index=$Emitter.SpawnGroup, Num=::NewEnumerator0)
    UseOwnerScale=true
   )
   &AddVelocity(
    Cone Angle=&MergeHelper_FloatSwitch(0=35, 1=42, 2=?, 3=?, Index=$Emitter.SpawnGroup, Num=::NewEnumerator0)
    Distribution Along Cone Axis:NiagaraFloat=&MergeHelper_FloatSwitch(0=0.5, 1=1, 2=?, 3=?, Index=$Emitter.SpawnGroup, Num=::NewEnumerator0)
    Inner Cone Angle:NiagaraFloat=&MergeHelper_FloatSwitch(0=30, 1=0, 2=?, 3=?, Index=$Emitter.SpawnGroup, Num=::NewEnumerator0)
    Module.UseConeFalloff=true
    Module.UseVelDistribution=true
    Override Seed=false
    Speed Falloff From Cone Axis:NiagaraFloat=&MergeHelper_FloatSwitch(0=0.5, 1=1, 2=?, 3=?, Index=$Emitter.SpawnGroup, Num=::NewEnumerator0)
    Velocity Mode=::NewEnumerator2
    Velocity Speed=&MergeHelper_FloatSwitch(
     0=-700
     1=&RandomRangeFloat(Minimum=-700, Maximum=-950)
     2=?
     3=?
     Index=$Emitter.SpawnGroup
     Num=::NewEnumerator0
    )
   )
   &DLR_ProductPreviews(GameModeID:NiagaraInt32=$System.GameModeID, bPreviewVehicleSpeed:NiagaraBool=$System.bPreviewVehicleSpeed)
  ]
  particleUpdateScript [
   &ParticleState()
   &ScaleSpriteSize(Uniform Curve Sprite Scale=DataInterface() {propertyValues {Curve {
       defaultValue=1
       keys [
        {arriveTangent=2.1099, interpMode=RCIM_Cubic, leaveTangent=2.1099, tangentMode=RCTM_User, time=0, value=1}
        {
         arriveTangent=0.6028
         arriveTangentWeight=0.1776
         interpMode=RCIM_Cubic
         leaveTangent=0.6028
         leaveTangentWeight=0.1776
         tangentMode=RCTM_User
         tangentWeightMode=RCTWM_WeightedBoth
         time=1
         value=1.5
        }
       ]
      }}})
   &SpriteRotationRate(Rotation Rate=&MergeHelper_FloatSwitch(0=100, 1=0, 2=?, 3=?, Index=&SpawnGroup(), Num=::NewEnumerator0))
   &AccelerationForce(Acceleration=&MergeHelper_VectorSwitch(
     0=<0 0 0>
     1=&MakeVector(
      X=0
      Y=&FloatFromCurve(
       FloatCurve=DataInterface() {propertyValues {Curve {defaultValue=0, keys [<at 0 0 RCIM_Constant>,<at 0.4 1 RCIM_Constant>]}}}
       Scale Curve=&SelectFloatFromArray(
        Array Sampling Mode=::NewEnumerator0
        Direct Array Index=?
        Float Selection Array=DataInterface() {propertyValues {FloatData [-800,800]}}
        Random Seed=0
       )
      )
      Z=-50
     )
     2=?
     3=?
     Index=&SpawnGroup()
     Num=::NewEnumerator0
    ))
   &Drag(
    Drag=&MergeHelper_FloatSwitch(0=1, 1=1.2, 2=?, 3=?, Index=&SpawnGroup(), Num=::NewEnumerator0)
    Use Linear Drag=true
    Use Rotational Drag=false
   )
   &InheritSourceMovement(
    Applied Position Scale=&Multiply_VectorByFloat(
     Vector=<1 1 1>
     Float=&FloatFromCurve(FloatCurve=DataInterface() {propertyValues {Curve {
         defaultValue=1
         keys [
          {arriveTangent=-1.7885, interpMode=RCIM_Cubic, leaveTangent=-1.7885, tangentMode=RCTM_User, time=0.05, value=1}
          {interpMode=RCIM_Cubic, tangentMode=RCTM_User, time=1, value=0}
         ]
        }}})
    )
    ApplyForce=false
    ApplyPosition=true
    ApplyPositionFalloff=false
   ) {enabled=false, module=InheritSourceMovement001}
   &InheritSourceMovement(ApplyForce=false, ApplyPosition=true, ApplyPositionFalloff=false)
   &SolveForcesAndVelocity(Clamp Velocity=false, Limit Acceleration=false)
   &ScaleColor(
    Scale Alpha=&FloatFromCurve(FloatCurve=DataInterface() {propertyValues {Curve {keys [<at 0.06 0>,<at 0.2 1>]}}})
    ScaleA=true
    ScaleRGB=true
   )
   /SetVariables(
    Particles.StarBack:Vector2f=&Multiply_Vector2D_ByFloat(
     Float=&FloatFromCurve(
      FloatCurve=DataInterface() {propertyValues {Curve {
         defaultValue=1
         keys [
          {arriveTangent=4.4693, interpMode=RCIM_Cubic, leaveTangent=4.4693, tangentMode=RCTM_User, time=-0.04, value=0}
          {arriveTangent=0.0366, interpMode=RCIM_Cubic, leaveTangent=0.0366, leaveTangentWeight=0.0836, tangentMode=RCTM_User, time=0.75, value=1}
          {arriveTangent=-9.0725, arriveTangentWeight=0.7606, interpMode=RCIM_Cubic, leaveTangent=-9.0725, tangentMode=RCTM_User, time=0.95, value=0}
         ]
        }}}
      Scale Curve=1.3
     )
     Vector2D=<32 32>
    )
    Particles.StarFront:Vector2f=&Multiply_Vector2D_ByFloat(
     Float=&FloatFromCurve(FloatCurve=DataInterface() {propertyValues {Curve {
         defaultValue=1
         keys [
          {arriveTangent=4.4693, interpMode=RCIM_Cubic, leaveTangent=4.4693, tangentMode=RCTM_User, time=-0.05, value=0}
          {arriveTangent=0.0366, interpMode=RCIM_Cubic, leaveTangent=0.0366, leaveTangentWeight=0.0836, tangentMode=RCTM_User, time=0.75, value=1}
          {arriveTangent=-9.0725, arriveTangentWeight=0.7606, interpMode=RCIM_Cubic, leaveTangent=-9.0725, tangentMode=RCTM_User, time=0.95, value=0}
         ]
        }}})
     Vector2D=<30 30>
    )
   ) {isSetParameters=true}
   /SetVariables(
    Particles.DitherOutline:LinearColor=&MakeLinearColorFromVectorAndFloat(
     "Vector (RGB)"=&MakeVectorFromLinearColorRGB(Color=$Particles.R1 Color)
     "Float (Alpha)"=&FloatFromCurve(FloatCurve=DataInterface() {propertyValues {Curve {keys [<at 0.06 0>,<at 0.2 1>]}}})
    )
    Particles.OutlineOffset=1
   ) {isSetParameters=true}
   &SubUVAnimation(
    "Sprite Renderer "=DataInterface() {propertyValues {SpriteRenderer=&self:NE_VC_Boost_Sprite_Parent001.NiagaraSpriteRendererProperties_0}}
    UseEndFrame=false
    UseStartFrame=false
   )
   &SubUVAnimation(
    "Sprite Renderer "=DataInterface() {propertyValues {SpriteRenderer=&self:NE_VC_Boost_Sprite_Parent001.NiagaraSpriteRendererProperties_1}}
    UseEndFrame=false
    UseStartFrame=false
   ) {module=SubUVAnimation001}
  ]
  rendererClasses [&NiagaraSpriteRendererProperties]
  renderers [
   {rendererClass=&NiagaraSpriteRendererProperties}
   {rendererClass=&NiagaraSpriteRendererProperties, rendererIndex=1}
   {rendererClass=&NiagaraSpriteRendererProperties, rendererIndex=2}
   {rendererClass=&NiagaraSpriteRendererProperties, rendererIndex=3}
  ]
 }
}
summary {
 systemName=NS_Boost_Toon
 emitters [
  {emitterName=NE_VC_Boost_Ribbon_Flame, rendererClasses [&NiagaraRibbonRendererProperties]}
  {emitterName=NE_VC_Boost_Sprite_Parent001, rendererClasses [&NiagaraSpriteRendererProperties]}
  {emitterName=NE_VC_Boost_Sprite_Parent, rendererClasses [&NiagaraSpriteRendererProperties]}
 ]
 userVariables [
  User.bIsFrontendPreview {type=NiagaraBool, defaultValue=true}
  User.bIsPainted {type=NiagaraBool}
  User.ColorVariantID {type=NiagaraInt32}
  User.DTCarOffset {type=NiagaraPosition, description="", defaultValue {}}
  User.PaintedColor {type=LinearColor, description="", defaultValue {struct=LinearColor, value=<rgba 1 1 1 1>}}
  User.PaintedScalar {type=NiagaraFloat, defaultValue=1, description=""}
  User.PortCount {type=NiagaraInt32, defaultValue=2}
 ]
}
